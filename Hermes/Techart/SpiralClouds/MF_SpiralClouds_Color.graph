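# Material function that computes spiral cloud color from a normals input and a halo mask, using the low-altitude clouds parameter collection.
# Half-Lambert sky-light shading blends shadow and light colors, then the halo mask blends toward a boosted halo color.
# A squared local-XY gradient controls desaturation and halo hue.
# Where that gradient is low, time-of-day tint and brightness from the camera-to-actor vertical direction apply.
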MaterialFunction /Hermes/Techart/SpiralClouds/MF_SpiralClouds_Color
graph {nodes [
  n1 = ComponentMask(None=CollectionParameter() {collection=/Game/Environments/Helios/TechArt/LowAltitudeClouds/V2/MPC_LowAltitudeClouds_V2, parameterName=ColorLight}) {b=true, g=true, r=true}
  n2 = Multiply(A=DotProduct(A=SkyAtmosphereLightDirection(), B=FunctionInput() {inputName=Normals})) {constB=0.5}
  n3 = LinearInterpolate(
   A=ComponentMask(None=CollectionParameter() {collection=&MPC_LowAltitudeClouds_V2, parameterName=ColorShadow}) {b=true, g=true, r=true}
   B=@n1
   Alpha=Power(Base=Add(A=@n2) {constB=0.5}) {constExponent=1.2}
  )
  n4 = Constant() {r=0.5}
  n5 = Multiply(A=DotProduct(A=LocalPosition().XY, B=Constant2Vector() {g=0.2, r=0.8})) {constB=-1}
  n6 = Saturate(None=Add(A=Divide(A=@n5) {constB=50000}) {constB=0.3})
  n7 = Multiply(A=@n6, B=@n6)
  n8 = ComponentMask(None=CollectionParameter() {collection=&MPC_LowAltitudeClouds_V2, parameterName=ColorHalo}) {b=true, g=true, r=true}
  n9 = LinearInterpolate(A=@n1, B=@n8)
  n10 = LinearInterpolate(
   A=LinearInterpolate(A=Desaturation(None=@n3, Fraction=@n4), B=@n3, Alpha=@n7)
   B=LinearInterpolate(A=Multiply(A=@n9) {constB=2}, B=Multiply(A=@n8) {constB=2}, Alpha=@n7)
   Alpha=Multiply(A=FunctionInput() {bUsePreviewValueAsDefault=true, inputName=Halo Mask, inputType=FunctionInput_Scalar, sortPriority=3}) {constB=2}
  )
  n11 = Multiply(
   A=@n10
   B=MaterialFunctionCall(
    Color_Morning=VectorParameter() {parameterName=Color Morning, defaultValue=<rgba 0.6816 0.7525 0.9323 1>}.RGB
    Color_Day=VectorParameter() {parameterName=Color Day, defaultValue=<rgba 0.3964 0.3334 0.4115 1>}.RGB
    Color_Evening=VectorParameter() {parameterName=Color Evening, defaultValue=<rgba 0.6021 0.5552 0.7865 1>}.RGB
    Color_Night=VectorParameter() {parameterName=Color Night, defaultValue=<rgba 0.95 0.9557 1 1>}.RGB
   ) {materialFunction=&Color_Time_of_Day}.Result
  ) {constB=0.5}
  n12 = Normalize(VectorInput=ActorPositionWS() {originType=CameraRelative})
  n13 = Add(A=Multiply(A=ComponentMask(None=@n12) {b=true}) {constB=0.4}) {constB=0.75}
  FunctionOutput(None=LinearInterpolate(A=Multiply(A=@n11, B=@n13), B=@n10, Alpha=@n7))
 ]}
