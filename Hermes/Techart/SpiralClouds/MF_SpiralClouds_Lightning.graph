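# Material function that generates animated, colored lightning for the spiral clouds effect.
# Two textures share slowly panning instance-space XY UVs: one is offset by per-channel time and thresholded into narrow traveling bands, which mask the other pattern.
# Each RGB channel pulses on its own triangle wave, then the three are summed into one intensity.
# That intensity is scaled by the Lightning Alpha input, boosted and tinted by the Lighting Color parameter.
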MaterialFunction /Hermes/Techart/SpiralClouds/MF_SpiralClouds_Lightning
graph {nodes [
  n1 = Time()
  n2 = Constant3Vector() {constant=<rgba 0.02 0.04 0.03 0>}
  n3 = TransformPosition(None=WorldPosition().XYZ) {transformSourceType=TRANSFORMPOSSOURCE_World, transformType=TRANSFORMPOSSOURCE_Instance}
  n4 = Add(A=Divide(A=ComponentMask(None=@n3) {g=true, r=true}) {constB=70000}, B=Multiply(A=Time(), B=Constant2Vector() {g=0.01, r=0.03}))
  n5 = Frac(None=Add(
    A=Multiply(A=@n1, B=@n2) {constB=0.005}
    B=TextureSample(UVs=@n4) {samplerType=SAMPLERTYPE_LinearColor, texture=&/T_SpiralCloud_LightningPattern2}.RGB
   ))
  n6 = Add(A=Abs(None=Add(A=@n5) {constB=-0.5})) {constB=-0.475}
  n7 = Frac(None=Multiply(A=Constant3Vector() {constant=<rgba 2 2.2 1.45 0>}, B=Time()) {constB=0.005})
  n8 = Multiply(A=Abs(None=Add(A=@n7) {constB=-0.5})) {constB=5}
  MaterialExpressionMultiply_15 = Multiply(
   A=Multiply(
    A=Saturate(None=@n6)
    B=TextureSample(UVs=@n4) {samplerType=SAMPLERTYPE_LinearColor, texture=&/T_SpiralCloud_LightningPattern}.RGB
   ) {constB=20}
   B=Saturate(None=Add(A=@n8) {constB=0.33})
  )
  n9 = Convert(None=@MaterialExpressionMultiply_15) {
   nodeName=Break Float3
   convertInputs [{
     type=Vector3
     expressionInput {
      expression=&self:MaterialExpressionMultiply_15
      inputName=None
      mask=0
      maskA=0
      maskB=0
      maskG=0
      maskR=0
      outputIndex=0
     }
    }]
   convertOutputs [{type=Scalar}, {type=Scalar}, {type=Scalar}]
  }
  n10 = FunctionInput() {bUsePreviewValueAsDefault=true, inputName=Lightning Alpha, inputType=FunctionInput_Scalar, sortPriority=2, previewValue=<1 1 1 1>}
  n11 = Multiply(A=Add(A=Add(A=@n9, B=@n9), B=@n9), B=@n10)
  FunctionOutput(None=Multiply(A=Multiply(A=@n11) {constB=10}, B=VectorParameter() {parameterName=Lighting Color, defaultValue=<rgba 1 0.1816 0.6597 1>}.RGB))
 ]}
