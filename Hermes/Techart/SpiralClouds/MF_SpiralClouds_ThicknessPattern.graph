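# Material function that produces an animated grayscale pattern for spiral clouds, named as their thickness pattern.
# It multiplies the U of the input texture coordinate by vertex color red, then scales it at two rates, scrolls each by time at different speeds and offsets both by vertex red.
# These scalar coordinates drive two samples of the spiral cloud mask, whose red channels are multiplied into the output.
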MaterialFunction /Hermes/Techart/SpiralClouds/MF_SpiralClouds_ThicknessPattern
graph {nodes [
  n1 = TextureCoordinate()
  n2 = Multiply(
   A=ComponentMask(None=FunctionInput(Preview=@n1) {bUsePreviewValueAsDefault=true, inputName=TexCoord, inputType=FunctionInput_Vector2}) {r=true}
   B=VertexColor().R
  )
  n3 = Time()
  n4 = VertexColor()
  n5 = Add(A=Add(A=Multiply(A=@n2) {constB=0.7}, B=Multiply(A=@n3) {constB=-0.014}), B=@n4.R)
  n6 = Time()
  n7 = Add(A=Add(A=Multiply(A=@n2) {constB=1.2}, B=Multiply(A=@n6) {constB=-0.01}), B=@n4.R)
  FunctionOutput(None=Multiply(
    A=TextureSample(UVs=@n5) {samplerType=SAMPLERTYPE_LinearGrayscale, texture=&/T_SpiralCloud_Mask}.R
    B=TextureSample(UVs=@n7) {samplerType=SAMPLERTYPE_LinearGrayscale, texture=&T_SpiralCloud_Mask}.R
   ))
 ]}
