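# Material function producing a procedural world-space XY spiral mask, filed under SpiralClouds.
# Custom HLSL computes approximate radial distance to logarithmic spiral arms around a center, fades it by Thickness, max-combines arms and multiplies by a soft Radius circle.
# A QuadSpiral static switch selects four arms or two; a single-arm node exists but is unconnected.
# Position defaults to world XY without shader offsets; SpiralRotation and SpiralCurve set the spiral's scale and growth.
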MaterialFunction /Hermes/Techart/SpiralClouds/MF_SpiralMask_WS
graph {nodes [
  MaterialExpressionFunctionInput_1 = FunctionInput(Preview=WorldPosition() {worldPositionShaderOffset=WPT_ExcludeAllShaderOffsets}.XY) {bUsePreviewValueAsDefault=true, inputName=PositionXY, inputType=FunctionInput_Vector2, sortPriority=-1}
  MaterialExpressionFunctionInput_6 = FunctionInput(Preview=Constant2Vector()) {bUsePreviewValueAsDefault=true, inputName=SpiralCenterXY, inputType=FunctionInput_Vector2}
  MaterialExpressionFunctionInput_0 = FunctionInput(Preview=Constant() {r=2}) {bUsePreviewValueAsDefault=true, inputName=SpiralRotation, inputType=FunctionInput_Scalar, sortPriority=1}
  MaterialExpressionFunctionInput_2 = FunctionInput(Preview=Constant() {r=0.5}) {bUsePreviewValueAsDefault=true, inputName=SpiralCurve, inputType=FunctionInput_Scalar, sortPriority=2}
  MaterialExpressionFunctionInput_3 = FunctionInput(Preview=Constant() {r=20}) {bUsePreviewValueAsDefault=true, inputName=Thickness, inputType=FunctionInput_Scalar, sortPriority=3}
  MaterialExpressionFunctionInput_4 = FunctionInput(Preview=Constant() {r=100}) {bUsePreviewValueAsDefault=true, inputName=Radius, inputType=FunctionInput_Scalar, sortPriority=4}
  Custom!(
   p=@MaterialExpressionFunctionInput_1
   spos=@MaterialExpressionFunctionInput_6
   a=@MaterialExpressionFunctionInput_0
   b=@MaterialExpressionFunctionInput_2
   d=@MaterialExpressionFunctionInput_3
   dr=@MaterialExpressionFunctionInput_4
  ) {
   code="// Spiral SDF\r\n// calculate the target radius and theta\r\nfloat2 pos = p - spos;\r\nfloat r = sqrt(pos.x*pos.x + pos.y*pos.y);\r\nfloat t = atan2Fast(pos.y, pos.x);\r\n\r\n//early exit if the point requested is the origin itself\r\n//to avoid taking the logarithm of zero in the next step\r\nif (r == 0) {return 0;}\r\n\r\n//calculate the floating point approximation for n\r\nfloat n = (log(r/a)/b - t)/(2.0*PI);\r\n\r\n//find the two possible radii for the closest point\r\nfloat upper_r = a * pow(exp(1), b * (t + 2.0*PI*ceil(n)));\r\nfloat lower_r = a * pow(exp(1), b * (t + 2.0*PI*floor(n)));\r\n\r\n//return the minimum distance to the target point\r\nfloat arc = min(abs(upper_r - r), abs(r - lower_r));\r\narc = 1.0f - saturate(arc/d);\r\n\r\n//calculate sphere mask\r\nfloat2 q = p - spos;\r\nfloat sphere = length(q) - dr;\r\nsphere = 1.0f - saturate(sphere/d);\r\n\r\nreturn sphere*arc;"
   description=Spiral SDF
   outputType=CMOT_Float1
   showCode=true
   inputs [
    {inputName=p, input {expression=&self:MaterialExpressionFunctionInput_1}}
    {inputName=spos, input {expression=&self:MaterialExpressionFunctionInput_6}}
    {inputName=a, input {expression=&self:MaterialExpressionFunctionInput_0}}
    {inputName=b, input {expression=&self:MaterialExpressionFunctionInput_2}}
    {inputName=d, input {expression=&self:MaterialExpressionFunctionInput_3}}
    {inputName=dr, input {expression=&self:MaterialExpressionFunctionInput_4}}
   ]
  }
  n1 = Custom(
   p=@MaterialExpressionFunctionInput_1
   spos=@MaterialExpressionFunctionInput_6
   a=@MaterialExpressionFunctionInput_0
   b=@MaterialExpressionFunctionInput_2
   d=@MaterialExpressionFunctionInput_3
   dr=@MaterialExpressionFunctionInput_4
  ) {
   code="// Quad Spiral SDF\r\n// calculate the target radius and theta\r\nfloat2 pos = p - spos;\r\nfloat r = sqrt(pos.x*pos.x + pos.y*pos.y);\r\nfloat t1 = atan2Fast(pos.y, pos.x);\r\nfloat t2 = t1+0.5*PI;\r\nfloat t3 = t1+PI;\r\nfloat t4 = t1+1.5*PI;\r\n\r\n//early exit if the point requested is the origin itself\r\n//to avoid taking the logarithm of zero in the next step\r\nif (r == 0) {return 0;}\r\n\r\n//calculate the floating point approximation for n\r\nfloat n1 = (log(r/a)/b - t1)/(2.0*PI);\r\nfloat n2 = (log(r/a)/b - t2)/(2.0*PI);\r\nfloat n3 = (log(r/a)/b - t3)/(2.0*PI);\r\nfloat n4 = (log(r/a)/b - t4)/(2.0*PI);\r\n\r\n//find the two possible radii for the closest point\r\nfloat upper_r1 = a * pow(exp(1), b * (t1 + 2.0*PI*ceil(n1)));\r\nfloat lower_r1 = a * pow(exp(1), b * (t1 + 2.0*PI*floor(n1)));\r\n\r\nfloat upper_r2 = a * pow(exp(1), b * (t2 + 2.0*PI*ceil(n2)));\r\nfloat lower_r2 = a * pow(exp(1), b * (t2 + 2.0*PI*floor(n2)));\r\n\r\nfloat upper_r3 = a * pow(exp(1), b * (t3 + 2.0*PI*ceil(n3)));\r\nfloat lower_r3 = a * pow(exp(1), b * (t3 + 2.0*PI*floor(n3)));\r\n\r\nfloat upper_r4 = a * pow(exp(1), b * (t4 + 2.0*PI*ceil(n4)));\r\nfloat lower_r4 = a * pow(exp(1), b * (t4 + 2.0*PI*floor(n4)));\r\n\r\n//return the minimum distance to the target point\r\nfloat arc1 = min(abs(upper_r1 - r), abs(r - lower_r1));\r\narc1 = 1.0f - saturate(arc1/d);\r\n\r\nfloat arc2 = min(abs(upper_r2 - r), abs(r - lower_r2));\r\narc2 = 1.0f - saturate(arc2/d);\r\n\r\nfloat arc3 = min(abs(upper_r3 - r), abs(r - lower_r3));\r\narc3 = 1.0f - saturate(arc3/d);\r\n\r\nfloat arc4 = min(abs(upper_r4 - r), abs(r - lower_r4));\r\narc4 = 1.0f - saturate(arc4/d);\r\n\r\nfloat arc = max(arc1, arc2);\r\narc = max(arc, arc3);\r\narc = max(arc, arc4);\r\n\r\n//calculate sphere mask\r\nfloat2 q = p - spos;\r\nfloat sphere = length(q) - dr;\r\nsphere = 1.0f - saturate(sphere/d);\r\n\r\nreturn sphere*arc;"
   description=Quad Spiral SDF
   outputType=CMOT_Float1
   showCode=true
   inputs [
    {inputName=p, input {expression=&self:MaterialExpressionFunctionInput_1}}
    {inputName=spos, input {expression=&self:MaterialExpressionFunctionInput_6}}
    {inputName=a, input {expression=&self:MaterialExpressionFunctionInput_0}}
    {inputName=b, input {expression=&self:MaterialExpressionFunctionInput_2}}
    {inputName=d, input {expression=&self:MaterialExpressionFunctionInput_3}}
    {inputName=dr, input {expression=&self:MaterialExpressionFunctionInput_4}}
   ]
  }
  n2 = Custom(
   p=@MaterialExpressionFunctionInput_1
   spos=@MaterialExpressionFunctionInput_6
   a=@MaterialExpressionFunctionInput_0
   b=@MaterialExpressionFunctionInput_2
   d=@MaterialExpressionFunctionInput_3
   dr=@MaterialExpressionFunctionInput_4
  ) {
   code="// Double Spiral SDF\r\n// calculate the target radius and theta\r\nfloat2 pos = p - spos;\r\nfloat r = sqrt(pos.x*pos.x + pos.y*pos.y);\r\nfloat t1 = atan2Fast(pos.y, pos.x);\r\nfloat t2 = t1+PI;\r\n\r\n//early exit if the point requested is the origin itself\r\n//to avoid taking the logarithm of zero in the next step\r\nif (r == 0) {return 0;}\r\n\r\n//calculate the floating point approximation for n\r\nfloat n1 = (log(r/a)/b - t1)/(2.0*PI);\r\nfloat n2 = (log(r/a)/b - t2)/(2.0*PI);\r\n\r\n//find the two possible radii for the closest point\r\nfloat upper_r1 = a * pow(exp(1), b * (t1 + 2.0*PI*ceil(n1)));\r\nfloat lower_r1 = a * pow(exp(1), b * (t1 + 2.0*PI*floor(n1)));\r\n\r\nfloat upper_r2 = a * pow(exp(1), b * (t2 + 2.0*PI*ceil(n2)));\r\nfloat lower_r2 = a * pow(exp(1), b * (t2 + 2.0*PI*floor(n2)));\r\n\r\n//return the minimum distance to the target point\r\nfloat arc1 = min(abs(upper_r1 - r), abs(r - lower_r1));\r\narc1 = 1.0f - saturate(arc1/d);\r\n\r\nfloat arc2 = min(abs(upper_r2 - r), abs(r - lower_r2));\r\narc2 = 1.0f - saturate(arc2/d);\r\n\r\nfloat arc = max(arc1, arc2);\r\n\r\n//calculate sphere mask\r\nfloat2 q = p - spos;\r\nfloat sphere = length(q) - dr;\r\nsphere = 1.0f - saturate(sphere/d);\r\n\r\nreturn sphere*arc;"
   description=Double Spiral SDF
   outputType=CMOT_Float1
   showCode=true
   inputs [
    {inputName=p, input {expression=&self:MaterialExpressionFunctionInput_1}}
    {inputName=spos, input {expression=&self:MaterialExpressionFunctionInput_6}}
    {inputName=a, input {expression=&self:MaterialExpressionFunctionInput_0}}
    {inputName=b, input {expression=&self:MaterialExpressionFunctionInput_2}}
    {inputName=d, input {expression=&self:MaterialExpressionFunctionInput_3}}
    {inputName=dr, input {expression=&self:MaterialExpressionFunctionInput_4}}
   ]
  }
  n3 = StaticBool()
  FunctionOutput(None=StaticSwitch(True=@n1, False=@n2, Value=FunctionInput(Preview=@n3) {inputName=QuadSpiral, inputType=FunctionInput_StaticBool, sortPriority=5}))
 ]}
