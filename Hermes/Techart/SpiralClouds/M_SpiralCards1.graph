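# Unlit masked material for spiral cloud cards, colored by volume-slice normals and a sun-facing halo, with added lightning.
# World position offset expands collapsed cards along tangent-based axes, shaped by thickness and wobble patterns and camera distance.
# Opacity combines noisy soft card shapes, distance fade and global collection opacity through temporal dithering, optionally masked by spiral growth.
# The spiral mask uses a material function, replaced by a cheap custom radial mask on Low quality.
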Material /Hermes/Techart/SpiralClouds/M_SpiralCards1
graph {
 properties {bAutomaticallySetUsageInEditor=false, blendMode=BLEND_Masked, shadingModel=MSM_Unlit, usedShadingModels=Unlit}
 nodes [
  n1 = Add(A=LocalPosition().XYZ, B=NamedRerouteUsage())
  n2 = MaterialFunctionCall!(
   UVs=VertexInterpolator(VS=Divide(A=@n1) {constB=6000}).PS
   Slices Texture=TextureObject() {samplerType=SAMPLERTYPE_LinearColor, texture=/Game/Environments/Helios/TechArt/LowAltitudeClouds/T_LowAltitudeClouds_N}
   Atlas Size=Constant() {r=8}
   Atlas Padding=ScalarParameter() {defaultValue=0.9375, parameterName=AtlasPadding, primitiveDataIndex=8}
  ) {materialFunction=&/MF_SpiralClouds_3DVolumeFrom2DSlices}
  n3 = Multiply(A=Add(A=@n2.RGB) {constB=-0.5}) {constB=3}
  n4 = Divide(
   A=Subtract(A=NamedRerouteUsage(), B=NamedRerouteUsage())
   B=Multiply(A=ScalarParameter() {defaultValue=1400, parameterName=Size}) {constB=0.4}
  ) {constB=50}
  n5 = VertexInterpolator(VS=@n4)
  n6 = CameraVectorWS()
  n7 = Saturate(None=DotProduct(A=Multiply(A=@n6) {constB=-1}, B=SkyAtmosphereLightDirection()))
  n8 = NamedRerouteUsage()
  n9 = Multiply(A=Multiply(A=@n7, B=@n7), B=OneMinus(None=Multiply(A=@n8) {constB=0.4}))
  n10 = Multiply(A=@n9, B=@n9)
  n11 = NamedRerouteUsage()
  n12 = VertexInterpolator(VS=OneMinus(None=Multiply(A=@n11) {constB=0.65}))
  n13 = Multiply(A=Power(Base=NamedRerouteUsage()) {constExponent=4}) {constB=0.85}
  Add(
   A=MaterialFunctionCall!(Normals=Add(A=@n3, B=@n5.PS), Halo Mask=Multiply(A=@n10, B=@n12.PS)) {materialFunction=&/MF_SpiralClouds_Color}.Result
   B=MaterialFunctionCall!(Lightning Alpha=OneMinus(None=@n13)) {materialFunction=&/MF_SpiralClouds_Lightning}.Result
  )
  n14 = Add(A=@n4, B=WorldPosition() {worldPositionShaderOffset=WPT_CameraRelative}.XYZ)
  Add(A=@n4, B=Multiply(A=Normalize(VectorInput=@n14)) {constB=-0.5})
  n15 = Multiply(A=Add(A=TextureCoordinate()) {constB=-0.5}) {constB=4}
  MaterialExpressionMultiply_10 = Multiply(A=@n15, B=@n15)
  n16 = Convert(None=@MaterialExpressionMultiply_10) {
   nodeName=Break Float2
   convertInputs [{
     type=Vector2
     expressionInput {
      expression=&self:MaterialExpressionMultiply_10
      inputName=None
      mask=0
      maskA=0
      maskB=0
      maskG=0
      maskR=0
      outputIndex=0
     }
    }]
   convertOutputs [{type=Scalar}, {type=Scalar}]
  }
  n17 = Subtract(A=Add(A=@n16, B=@n16)) {constB=0.5}
  n18 = VertexInterpolator(VS=OneMinus(None=Saturate(None=@n17)))
  n19 = Add(A=TextureCoordinate()) {constB=-0.5}
  n20 = ComponentMask(None=@n19) {g=true}
  n21 = NamedRerouteUsage()
  n22 = Multiply(A=AppendVector(A=ComponentMask(None=@n19) {r=true}, B=Multiply(A=@n20, B=@n21)), B=Constant2Vector() {g=2, r=1})
  n23 = Multiply(A=Time() {period=1}) {constB=-0.1}
  n24 = Multiply(A=TextureCoordinate() {coordinateIndex=1}) {constB=32156}
  n25 = VertexInterpolator(VS=Add(A=Subtract(A=@n22, B=@n23), B=Frac(None=@n24)))
  n26 = Multiply(
   A=@n18.PS
   B=Add(
    A=@n18.PS
    B=TextureSample(UVs=@n25.PS) {samplerType=SAMPLERTYPE_LinearGrayscale, texture=/Game/Environments/Helios/TechArt/LowAltitudeClouds/T_LowAltitudeClouds_A}.R
   )
  )
  n27 = NamedRerouteUsage()
  n28 = Multiply(A=Subtract(A=Multiply(A=@n27) {constB=4}) {constB=0.7}, B=Multiply(A=NamedRerouteUsage()) {constB=2})
  n29 = WorldPosition() {worldPositionShaderOffset=WPT_CameraRelative}
  n30 = Constant()
  n31 = Subtract(A=Divide(A=Distance(A=@n29.XYZ, B=@n30)) {constB=4000}) {constB=0.5}
  n32 = Multiply(A=Saturate(None=Multiply(A=@n26) {constB=5}), B=Multiply(A=Saturate(None=@n28), B=Saturate(None=@n31)))
  Multiply(
   A=Multiply(
    A=MaterialFunctionCall!(Alpha Threshold=@n32, Random) {materialFunction=/Game/Environments/Asteria/Sets/NeonCity/Techart/Screen/MaterialFunctions/DitherTemporalAABlue}.Result
    B=CollectionParameter() {collection=/Game/Environments/Helios/TechArt/LowAltitudeClouds/V2/MPC_LowAltitudeClouds_V2, parameterName=GlobalOpacity}
   )
   B=StaticSwitchParameter(True=NamedRerouteUsage(), False=Constant() {r=1}) {parameterName=SpiralGrowthMask}
  )
  n33 = NamedRerouteUsage()
  n34 = Add(A=@n33, B=MaterialFunctionCall!(WPO=@n33) {materialFunction=&/MF_SpiralClouds_WobblePattern}.Result)
  n35 = Add(A=TextureCoordinate()) {constB=-0.5}
  n36 = Multiply(A=ComponentMask(None=Abs(None=@n35)) {g=true}) {constB=2}
  n37 = Multiply(A=OneMinus(None=Multiply(A=@n36) {constB=0.5}), B=NamedRerouteUsage())
  n38 = NamedRerouteDeclaration(None=Multiply(A=NamedRerouteUsage(), B=NamedRerouteUsage())) {name=Length * Thickness Mask, nodeColor=<rgba 1 0.6063 0.1807 1>}
  n39 = NamedRerouteUsage()
  n40 = Normalize(VectorInput=CrossProduct(A=VertexTangentWS(), B=@n39))
  n41 = Add(A=TextureCoordinate()) {constB=-0.5}
  n42 = ComponentMask(None=TextureCoordinate() {coordinateIndex=1}) {r=true}
  n43 = Multiply(A=Multiply(A=@n42, B=OneMinus(None=@n42))) {constB=2}
  n44 = Multiply(
   A=LinearInterpolate(
    A=ScalarParameter() {parameterName=Base Size1}
    B=ScalarParameter() {defaultValue=15000, parameterName=Base Size 2}
    Alpha=NamedRerouteDeclaration(None=@n43) {name=Length Mask, nodeColor=<rgba 0 1 0.1797 1>}
   )
   B=VertexColor().G
  )
  n45 = Multiply(A=Multiply(A=@n40, B=ComponentMask(None=@n41) {g=true}), B=@n44)
  n46 = Subtract(A=MaterialFunctionCall!(TexCoord=TextureCoordinate() {coordinateIndex=1}) {materialFunction=&/MF_SpiralClouds_ThicknessPattern}.Result) {constB=0.1}
  n47 = NamedRerouteDeclaration(None=Saturate(None=Multiply(A=@n46) {constB=1.2})) {name=Thickness Pattern, nodeColor=<rgba 0 1 0.8125 1>}
  n48 = CrossProduct(A=@n40, B=@n39)
  n49 = Multiply(A=ComponentMask(None=@n41) {r=true}) {constB=-1}
  n50 = VertexTangentWS()
  n51 = Multiply(A=NamedRerouteUsage()) {constB=-1}
  n52 = NamedRerouteDeclaration(None=Abs(None=DotProduct(A=@n50, B=@n51))) {name=Fresnel, nodeColor=<rgba 0 0.5547 1 1>}
  n53 = Multiply(A=Multiply(A=Multiply(A=@n48, B=@n49), B=@n44), B=LinearInterpolate(B=@n47, Alpha=Power(Base=@n52) {constExponent=32}) {constA=1})
  n54 = Add(A=Multiply(A=Multiply(A=@n37, B=@n38)) {constB=-4000}, B=Add(A=Multiply(A=@n45, B=@n47), B=@n53))
  n55 = Distance(A=NamedRerouteUsage(), B=Constant())
  n56 = Saturate(None=Subtract(A=Divide(A=@n55) {constB=5000}) {constB=0.5})
  NamedRerouteDeclaration(None=Add(A=NamedRerouteDeclaration(None=@n34) {name=Rotating Collapsed Cards, nodeColor=<rgba 0 0.6953 1 1>}, B=Multiply(A=@n54, B=@n56))) {name=WPO, nodeColor=<rgba 0.125 1 0 1>}
  NamedRerouteDeclaration(None=@n2.A) {name=Alpha, nodeColor=<rgba 1 0.3047 0 1>}
  n57 = AppendVector(A=@n35, B=Constant())
  n58 = ScalarParameter() {defaultValue=16.6586, parameterName=Length 1}
  n59 = ScalarParameter() {defaultValue=-12.5399, parameterName=Length 2}
  n60 = NamedRerouteUsage()
  n61 = AppendVector(A=Multiply(A=LinearInterpolate(A=@n58, B=@n59, Alpha=@n60) {constA=6, constB=3}) {constB=200}, B=Constant2Vector() {g=-200, r=-200})
  NamedRerouteDeclaration(None=Transform(None=Multiply(A=@n57, B=@n61) {constB=-200})) {name=Collapsed Cards, nodeColor=<rgba 0 1 0.6484 1>}
  MaterialExpressionWorldPosition_5 = WorldPosition() {worldPositionShaderOffset=WPT_ExcludeAllShaderOffsets}
  MaterialExpressionComponentMask_10 = ComponentMask(None=ActorPositionWS()) {g=true, r=true}
  MaterialExpressionScalarParameter_33 = ScalarParameter() {defaultValue=10000, parameterName=d}
  MaterialExpressionScalarParameter_34 = ScalarParameter() {defaultValue=31850.1406, parameterName=dr}
  n62 = StaticBool()
  n63 = StaticBoolParameter() {defaultValue=true, parameterName=QuadSpiral}
  n64 = MaterialFunctionCall!(
   PositionXY=@MaterialExpressionWorldPosition_5.XY
   SpiralCenterXY=@MaterialExpressionComponentMask_10
   SpiralRotation=ScalarParameter() {defaultValue=1, parameterName=a}
   SpiralCurve=ScalarParameter() {defaultValue=0.03, parameterName=b}
   Thickness=@MaterialExpressionScalarParameter_33
   Radius=@MaterialExpressionScalarParameter_34
   QuadSpiral=QualitySwitch(Default=@n62, Low=@n62, High=@n63, Medium=@n62, Epic=@n63)
  ) {materialFunction=&/MF_SpiralMask_WS}
  n65 = Custom!(
   p=@MaterialExpressionWorldPosition_5.XY
   spos=@MaterialExpressionComponentMask_10
   r=@MaterialExpressionScalarParameter_34
   d=@MaterialExpressionScalarParameter_33
  ) {
   code="//calculate sphere mask\r\nfloat2 q = p - spos;\r\nfloat sphere = length(q) - r;\r\nsphere = 1.0f - saturate(sphere/d);\r\nreturn sphere;"
   outputType=CMOT_Float1
   showCode=true
   inputs [
    {
     inputName=p
     input {expression=&self:MaterialExpressionWorldPosition_5, mask=1, maskG=1, maskR=1, outputIndex=1}
    }
    {inputName=spos, input {expression=&self:MaterialExpressionComponentMask_10}}
    {inputName=r, input {expression=&self:MaterialExpressionScalarParameter_34}}
    {inputName=d, input {expression=&self:MaterialExpressionScalarParameter_33}}
   ]
  }
  NamedRerouteDeclaration(None=QualitySwitch(Default=@n64.Result, Low=@n65, High=@n64.Result, Medium=@n64.Result, Epic=@n64.Result)) {name=SpiralMask, nodeColor=<rgba 0 0.5547 1 1>}
  NamedRerouteDeclaration(None=Add(A=Multiply(A=@n53, B=@n56), B=NamedRerouteUsage())) {name=Squashed WPO, nodeColor=<rgba 1 0.9844 0 1>}
  n66 = Add(A=@n34, B=WorldPosition() {worldPositionShaderOffset=WPT_CameraRelative}.XYZ)
  NamedRerouteDeclaration(None=Normalize(VectorInput=NamedRerouteDeclaration(None=@n66) {name=Card Position, nodeColor=<rgba 1 0.7969 0 1>})) {name=Card Vector, nodeColor=<rgba 0 1 0.5313 1>}
  NamedRerouteUsage()
 ]
}
