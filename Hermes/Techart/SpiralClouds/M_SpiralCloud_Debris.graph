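# Translucent unlit material for debris cards swirling in the spiral cloud effect, shaded by shared SpiralClouds color and lightning functions.
# UV1.x seeds per-card randomness for atlas tile, size, spin and orbit phase.
# WPO builds spinning view-facing cards on looping paths that spiral inward and down around Z, plus wobble.
# Normals are rebuilt to follow card spin, a sun-facing halo mask feeds color, and a global collection parameter scales opacity.
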Material /Hermes/Techart/SpiralClouds/M_SpiralCloud_Debris
graph {
 properties {
  bAutomaticallySetUsageInEditor=false
  bTangentSpaceNormal=false
  blendMode=BLEND_Translucent
  shadingModel=MSM_Unlit
  translucencyLightingMode=TLM_Surface
  twoSided=true
  usedShadingModels=Unlit
 }
 nodes [
  n1 = TextureCoordinate()
  n2 = Multiply(A=ComponentMask(None=TextureCoordinate() {coordinateIndex=1}) {r=true}) {constB=5.0778}
  n3 = Multiply(A=ComponentMask(None=TextureCoordinate() {coordinateIndex=1}) {r=true}) {constB=0.2576}
  n4 = AppendVector(A=Round(None=Frac(None=@n2)), B=Round(None=Frac(None=@n3)))
  n5 = VertexInterpolator(VS=Multiply(A=Add(A=@n1, B=@n4) {constB=0}) {constB=0.5})
  n6 = TextureSample(UVs=@n5.PS) {texture=&/T_SpiralCloud_Debris_D}
  n7 = VertexInterpolator(VS=NamedRerouteUsage())
  n8 = Rotator(Coordinate=Constant3Vector() {constant=<rgba 0 -1 0 0>}, Time=@n7.PS) {centerX=0, centerY=0, speed=1}
  n9 = Rotator(Coordinate=Constant3Vector() {constant=<rgba -1 0 0 0>}, Time=@n7.PS) {centerX=0, centerY=0, speed=1}
  n10 = CameraVectorWS()
  n11 = Saturate(None=DotProduct(A=Multiply(A=@n10) {constB=-1}, B=SkyAtmosphereLightDirection()))
  n12 = Multiply(A=Multiply(A=@n11, B=@n11)) {constB=0.5}
  n13 = MaterialFunctionCall!(
   Normals=MaterialFunctionCall(
    VectorToTransform=TextureSample(UVs=@n5.PS) {samplerType=SAMPLERTYPE_Normal, texture=&/T_SpiralCloud_Debris_N}.RGB
    BasisX=Transform(None=@n8) {transformSourceType=TRANSFORMSOURCE_View}
    BasisY=Transform(None=@n9) {transformSourceType=TRANSFORMSOURCE_View}
    BasisZ=CameraVectorWS()
   ) {materialFunction=&Transform3x3Matrix}.Transformed Vector
   Halo Mask=Multiply(A=@n12, B=@n12)
  ) {materialFunction=&/MF_SpiralClouds_Color}
  Add(
   A=Multiply(A=Multiply(A=@n6.RGB, B=@n13.Result), B=VectorParameter() {parameterName=Color, defaultValue=<rgba 2 3 3 1>}.RGB)
   B=MaterialFunctionCall!(Lightning Alpha=Constant() {r=1}) {materialFunction=&/MF_SpiralClouds_Lightning}.Result
  )
  Multiply(
   A=@n6.A
   B=CollectionParameter() {collection=/Game/Environments/Helios/TechArt/LowAltitudeClouds/V2/MPC_LowAltitudeClouds_V2, parameterName=GlobalOpacity}
  )
  n14 = Multiply(A=LocalPosition().XY, B=Constant2Vector() {g=-1, r=1})
  n15 = ComponentMask(None=TextureCoordinate() {coordinateIndex=1}) {r=true}
  n16 = NamedRerouteDeclaration(None=Frac(None=Multiply(A=@n15) {constB=1.9725})) {name=Random2, nodeColor=<rgba 0.8516 1 0 1>}
  n17 = NamedRerouteDeclaration(None=Multiply(A=Time(), B=LinearInterpolate(Alpha=@n16) {constA=-10, constB=10})) {nodeColor=<rgba 1 0 0.9375 1>}
  n18 = Transform(None=AppendVector(A=Rotator(Coordinate=@n14, Time=@n17) {centerX=0, centerY=0, speed=1}, B=Constant())) {transformSourceType=TRANSFORMSOURCE_View}
  n19 = Multiply(A=LinearInterpolate(Alpha=NamedRerouteUsage()) {constA=0.4, constB=3}, B=NamedRerouteUsage())
  n20 = Constant3Vector() {constant=<rgba 0 0 1 0>}
  n21 = Time()
  n22 = TextureCoordinate() {coordinateIndex=1}
  n23 = Frac(None=Add(A=Multiply(A=@n21) {constB=0.03}, B=ComponentMask(None=@n22) {r=true}) {constB=0.2322})
  n24 = OneMinus(None=@n23)
  n25 = SquareRoot(None=@n24)
  n26 = TextureCoordinate() {coordinateIndex=1}
  n27 = Frac(None=Multiply(A=ComponentMask(None=@n26) {r=true}) {constB=0.6763})
  n28 = Multiply(A=@n25, B=LinearInterpolate(Alpha=NamedRerouteDeclaration(None=@n27) {name=Random1, nodeColor=<rgba 1 0 0.2109 1>}) {constA=0.5, constB=1.5}) {constB=3}
  n29 = Constant3Vector()
  n30 = Add(A=@n24) {constB=0.1}
  n31 = LocalPosition()
  n32 = Multiply(A=Sign(None=Add(A=@n31.Z) {constB=-500})) {constB=19000}
  n33 = Constant()
  n34 = Multiply(A=@n24) {constB=4500}
  n35 = Add(A=AppendVector(A=Multiply(A=@n30, B=@n32) {constB=16000}, B=AppendVector(A=@n33, B=@n34)), B=Multiply(A=LocalPosition().XYZ) {constB=-1})
  n36 = Add(A=LocalPosition().XYZ, B=@n35)
  n37 = Transform(None=Add(A=RotateAboutAxis(NormalizedRotationAxis=@n20, RotationAngle=@n28, PivotPoint=@n29, Position=@n36), B=@n35)) {transformSourceType=TRANSFORMSOURCE_Local}
  n38 = MaterialFunctionCall!(WPO=@n37) {materialFunction=&/MF_SpiralClouds_WobblePattern}
  NamedRerouteDeclaration(None=Add(A=Multiply(A=@n18, B=@n19) {constB=2}, B=Add(A=@n37, B=@n38.Result))) {name=WPO, nodeColor=<rgba 1 0 0.5391 1>}
  NamedRerouteDeclaration(None=Multiply(A=SquareRoot(None=@n23), B=@n25)) {name=Fade In Out, nodeColor=<rgba 0 1 0.7422 1>}
  NamedRerouteUsage()
 ]
}
