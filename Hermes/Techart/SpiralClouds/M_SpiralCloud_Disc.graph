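# Translucent unlit material for an animated spiral cloud disc, with color and lightning from shared functions.
# Two scrolling overcast-cloud samples, using explicit raw-UV derivatives, blend by blue-channel difference to drive rotated normals and opacity.
# Radial coordinate, distortion, fresnel, sun halo and camera fade masks are named reroutes; opacity adds depth fade and global collection opacity.
# A world-space spiral mask drops to a cheap circular mask on Low quality.
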Material /Hermes/Techart/SpiralClouds/M_SpiralCloud_Disc
graph {
 properties {
  bAutomaticallySetUsageInEditor=false
  bTangentSpaceNormal=false
  blendMode=BLEND_Translucent
  shadingModel=MSM_Unlit
  usedShadingModels=Unlit
 }
 nodes [
  Add(A=NamedRerouteUsage()) {constB=-1}
  n1 = NamedRerouteUsage()
  Add(
   A=MaterialFunctionCall!(Normals=Multiply(A=@n1) {constB=2}, Halo Mask=NamedRerouteUsage()) {materialFunction=&/MF_SpiralClouds_Color}.Result
   B=MaterialFunctionCall!(Lightning Alpha=NamedRerouteUsage()) {materialFunction=&/MF_SpiralClouds_Lightning}.Result
  )
  MaterialExpressionAdd_14 = Add(A=TextureCoordinate()) {constB=-0.5}
  Convert(None=@MaterialExpressionAdd_14) {
   nodeName=Break Float2
   convertInputs [{
     type=Vector2
     expressionInput {
      expression=&self:MaterialExpressionAdd_14
      inputName=None
      mask=0
      maskA=0
      maskB=0
      maskG=0
      maskR=0
      outputIndex=0
     }
    }]
   convertOutputs [{type=Vector2}]
  }
  LocalPosition()
  MaterialExpressionAdd_18 = Add(A=TextureCoordinate()) {constB=-0.5}
  n2 = Convert(None=@MaterialExpressionAdd_18) {
   nodeName=Break Float2
   convertInputs [{
     type=Vector2
     expressionInput {
      expression=&self:MaterialExpressionAdd_18
      inputName=None
      mask=0
      maskA=0
      maskB=0
      maskG=0
      maskR=0
      outputIndex=0
     }
    }]
   convertOutputs [{type=Vector2}]
  }
  n3 = Multiply(A=@n2, B=Constant2Vector() {g=1, r=-1}) {constB=-1}
  n4 = Multiply(A=NamedRerouteUsage()) {constB=2}
  n5 = Multiply(A=Rotator(Coordinate=@MaterialExpressionAdd_14, Time=OneMinus(None=@n4)) {centerX=0, centerY=0, speed=-0.5}) {constB=-1}
  MaterialFunctionCall(
   BasisX=AppendVector(A=Normalize(VectorInput=@n3), B=Constant())
   BasisY=AppendVector(A=Normalize(VectorInput=@n5), B=Constant())
   BasisZ=Multiply(A=Constant3Vector() {constant=<rgba 0 0 1 0>}, B=TwoSidedSign()) {constB=-1}
  ) {materialFunction=&Transform3x3Matrix}
  n6 = NamedRerouteUsage()
  n7 = Constant() {r=1}
  Multiply(
   A=Multiply(A=NamedRerouteUsage(), B=StaticSwitchParameter(True=@n6, False=@n7) {parameterName=SpiralGrowthMask})
   B=ScalarParameter() {defaultValue=1, parameterName=OpacityMulti}
  )
  MaterialExpressionWorldPosition_0 = WorldPosition() {worldPositionShaderOffset=WPT_ExcludeAllShaderOffsets}
  MaterialExpressionComponentMask_0 = ComponentMask(None=ActorPositionWS()) {g=true, r=true}
  MaterialExpressionScalarParameter_33 = ScalarParameter() {defaultValue=10000, parameterName=d}
  MaterialExpressionScalarParameter_34 = ScalarParameter() {defaultValue=31850.1406, parameterName=dr}
  n8 = StaticBool()
  n9 = StaticBoolParameter() {defaultValue=true, parameterName=QuadSpiral}
  n10 = MaterialFunctionCall!(
   PositionXY=@MaterialExpressionWorldPosition_0.XY
   SpiralCenterXY=@MaterialExpressionComponentMask_0
   SpiralRotation=ScalarParameter() {defaultValue=1, parameterName=a}
   SpiralCurve=ScalarParameter() {defaultValue=0.03, parameterName=b}
   Thickness=@MaterialExpressionScalarParameter_33
   Radius=@MaterialExpressionScalarParameter_34
   QuadSpiral=QualitySwitch(Default=@n8, Low=@n8, High=@n9, Medium=@n8, Epic=@n9)
  ) {materialFunction=&/MF_SpiralMask_WS}
  n11 = Custom!(
   p=@MaterialExpressionWorldPosition_0.XY
   spos=@MaterialExpressionComponentMask_0
   r=@MaterialExpressionScalarParameter_34
   d=@MaterialExpressionScalarParameter_33
  ) {
   code="//calculate sphere mask\r\nfloat2 q = p - spos;\r\nfloat sphere = length(q) - r;\r\nsphere = 1.0f - saturate(sphere/d);\r\nreturn sphere;"
   outputType=CMOT_Float1
   showCode=true
   inputs [
    {
     inputName=p
     input {expression=&self:MaterialExpressionWorldPosition_0, mask=1, maskG=1, maskR=1, outputIndex=1}
    }
    {inputName=spos, input {expression=&self:MaterialExpressionComponentMask_0}}
    {inputName=r, input {expression=&self:MaterialExpressionScalarParameter_34}}
    {inputName=d, input {expression=&self:MaterialExpressionScalarParameter_33}}
   ]
  }
  NamedRerouteDeclaration(None=QualitySwitch(Default=@n10.Result, Low=@n11, High=@n10.Result, Medium=@n10.Result, Epic=@n10.Result)) {name=SpiralMask, nodeColor=<rgba 0 0.5547 1 1>}
  n12 = CameraVectorWS()
  n13 = Saturate(None=DotProduct(A=Multiply(A=@n12) {constB=-1}, B=SkyAtmosphereLightDirection()))
  n14 = Multiply(A=Multiply(A=Multiply(A=@n13, B=@n13), B=NamedRerouteUsage()), B=LinearInterpolate(Alpha=NamedRerouteUsage()) {constA=0.4})
  NamedRerouteDeclaration(None=Multiply(A=Multiply(A=@n14, B=@n14)) {constB=1.5}) {name=Halo Mask, nodeColor=<rgba 0 0.7656 1 1>}
  n15 = NamedRerouteUsage()
  n16 = Multiply(A=@n15) {constB=3}
  n17 = Multiply(A=Time()) {constB=-0.04}
  n18 = NamedRerouteUsage()
  n19 = Add(A=Add(A=@n15, B=Multiply(A=@n18) {constB=4}), B=Multiply(A=Time()) {constB=0.03})
  n20 = Multiply(A=ComponentMask(None=NamedRerouteUsage()) {r=true}) {constB=0.4}
  n21 = TextureCoordinate()
  n22 = DDX(Value=@n21)
  n23 = DDY(Value=@n21)
  n24 = TextureSample!(UVs=AppendVector(A=Add(A=@n16, B=@n17), B=Add(A=@n19, B=@n20)), Tex, "DDX(UVs)"=@n22, "DDY(UVs)"=@n23, Apply View MipBias) {
   mipValueMode=TMVM_Derivative
   samplerType=SAMPLERTYPE_LinearColor
   texture=/Game/TimeOfDay/TODM_Sequence/Materials/Sky/Textures/T_Ch6_OvercastClouds
  }
  n25 = NamedRerouteUsage()
  n26 = Multiply(A=@n25) {constB=2}
  n27 = Multiply(A=Time()) {constB=0.01}
  n28 = NamedRerouteUsage()
  n29 = Add(A=Add(A=@n25, B=Multiply(A=@n28) {constB=2}), B=Multiply(A=Time()) {constB=-0.01})
  n30 = Multiply(A=ComponentMask(None=NamedRerouteUsage()) {g=true}) {constB=0.3}
  n31 = TextureSample(UVs=AppendVector(A=Add(A=@n26, B=@n27), B=Add(A=@n29, B=@n30)), "DDX(UVs)"=@n22, "DDY(UVs)"=@n23) {
   mipValueMode=TMVM_Derivative
   samplerType=SAMPLERTYPE_LinearColor
   texture=&T_Ch6_OvercastClouds
  }
  n32 = Multiply(A=Subtract(A=@n31.B, B=@n24.B)) {constB=4}
  n33 = LinearInterpolate(A=@n24.RGBA, B=@n31.RGBA, Alpha=Saturate(None=Add(A=@n32) {constB=0.5}))
  n34 = Multiply(A=Add(A=ComponentMask(None=@n33) {g=true, r=true}) {constB=-0.5}) {constB=3}
  n35 = Rotator(Coordinate=@n34, Time=NamedRerouteUsage()) {centerX=0, centerY=0, speed=-6.28}
  n36 = Add(A=NamedRerouteUsage(), B=ScalarParameter() {defaultValue=3.7345})
  n37 = Multiply(A=ComponentMask(None=DeriveNormalZ(InXY=@n34)) {b=true}, B=ComponentMask(None=VertexNormalWS()) {b=true})
  n38 = ScalarParameter() {bUseCustomPrimitiveData=true, defaultValue=1, parameterName=Normal}
  NamedRerouteDeclaration(None=AppendVector(A=Rotator(Coordinate=@n35, Time=@n36) {centerX=0, centerY=0, speed=-1}, B=Multiply(A=@n37, B=@n38))) {name=Normals, nodeColor=<rgba 0.5 1 0 1>}
  n39 = Abs(None=DotProduct(A=CameraVectorWS(), B=VertexNormalWS()))
  NamedRerouteDeclaration(None=Saturate(None=Multiply(A=@n39) {constB=4})) {name=Fresnel, nodeColor=<rgba 0 0.6719 1 1>}
  n40 = Divide(A=Distance(A=WorldPosition() {worldPositionShaderOffset=WPT_CameraRelative}.XYZ, B=Constant())) {constB=3000}
  NamedRerouteDeclaration(None=Saturate(None=Subtract(A=@n40) {constB=0.5})) {name=Camera Fade, nodeColor=<rgba 1 0 0.1172 1>}
  n41 = Multiply(A=Distance(A=TextureCoordinate(), B=Constant() {r=0.5})) {constB=2}
  NamedRerouteDeclaration(None=SquareRoot(None=NamedRerouteDeclaration(None=@n41) {name=Sphere Mask, nodeColor=<rgba 1 0.3516 0 1>})) {name=Sphere Mask Squared, nodeColor=<rgba 1 0 0.9609 1>}
  MaterialExpressionAdd_5 = Add(A=TextureCoordinate()) {constB=-0.5}
  n42 = Convert(None=@MaterialExpressionAdd_5) {
   nodeName=Break Float2
   convertInputs [{
     type=Vector2
     expressionInput {
      expression=&self:MaterialExpressionAdd_5
      inputName=None
      mask=0
      maskA=0
      maskB=0
      maskG=0
      maskR=0
      outputIndex=0
     }
    }]
   convertOutputs [{type=Scalar}, {type=Scalar}]
  }
  n43 = Arctangent2Fast(Y=@n42, X=@n42)
  n44 = MaterialFunctionCall(Multiplier=Constant() {r=2}) {materialFunction=&Pi}
  NamedRerouteDeclaration(None=Frac(None=Divide(A=@n43, B=@n44.Result))) {name=Radial Coord, nodeColor=<rgba 0.9219 1 0 1>}
  n45 = Multiply(A=Rotator(Coordinate=TextureCoordinate()) {speed=-0.1}) {constB=0.5}
  NamedRerouteDeclaration(None=Add(A=TextureSample!(UVs=@n45, Tex, Apply View MipBias) {samplerType=SAMPLERTYPE_LinearColor, texture=&/T_SpiralCloud_Wave}.RGB) {constB=-0.5}) {name=Distortion, nodeColor=<rgba 0 1 0.6953 1>}
  n46 = ComponentMask(None=@n33) {b=true}
  n47 = OneMinus(None=@n46)
  NamedRerouteDeclaration(None=Multiply(A=@n47, B=@n47)) {name=Alpha, nodeColor=<rgba 0.5469 1 0 1>}
  n48 = Saturate(None=NamedRerouteUsage())
  n49 = Multiply(A=Multiply(A=OneMinus(None=@n48)) {constB=1.5}, B=Saturate(None=Multiply(A=@n48) {constB=7}))
  n50 = Multiply(A=Saturate(None=Multiply(A=@n49) {constB=2}), B=NamedRerouteUsage())
  n51 = Saturate(None=Multiply(A=Multiply(A=@n46, B=@n50) {constB=1.2}) {constB=1.2})
  n52 = Multiply(A=Multiply(A=@n51, B=@n51), B=NamedRerouteUsage())
  NamedRerouteDeclaration(None=Multiply(
    A=DepthFade(Opacity=@n52) {fadeDistanceDefault=3000}
    B=CollectionParameter() {collection=/Game/Environments/Helios/TechArt/LowAltitudeClouds/V2/MPC_LowAltitudeClouds_V2, parameterName=GlobalOpacity}
   )) {name=Opacity, nodeColor=<rgba 0.7344 0 1 1>}
  NamedRerouteUsage()
  VertexNormalWS()
 ]
}
