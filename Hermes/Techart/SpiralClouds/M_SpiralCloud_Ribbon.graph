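# Translucent unlit material for spiral cloud ribbons, expanding spline strips toward the camera via World Position Offset with thickness-driven width, wobble and camera offset.
# Two scrolling overcast-cloud samples are height-blended into normals and opacity, with edge masks, depth fade and global opacity.
# Normals and a sun halo feed a color function, added to a lightning function.
# A world-space spiral mask falls back to radial falloff on low quality.
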Material /Hermes/Techart/SpiralClouds/M_SpiralCloud_Ribbon
graph {
 properties {
  bAutomaticallySetUsageInEditor=false
  bTangentSpaceNormal=false
  blendMode=BLEND_Translucent
  shadingModel=MSM_Unlit
  usedShadingModels=Unlit
 }
 nodes [
  n1 = Subtract(A=Multiply(A=NamedRerouteUsage()) {constB=8}) {constB=3}
  n2 = Divide(A=Distance(A=WorldPosition() {worldPositionShaderOffset=WPT_CameraRelative}.XYZ, B=Constant())) {constB=3000}
  n3 = Multiply(A=Multiply(A=NamedRerouteUsage(), B=Saturate(None=@n1)), B=Saturate(None=Subtract(A=@n2)))
  Multiply(
   A=Multiply(
    A=DepthFade(Opacity=@n3) {fadeDistanceDefault=2000}
    B=CollectionParameter() {collection=/Game/Environments/Helios/TechArt/LowAltitudeClouds/V2/MPC_LowAltitudeClouds_V2, parameterName=GlobalOpacity}
   )
   B=StaticSwitchParameter(True=NamedRerouteUsage(), False=Constant() {r=1}) {parameterName=SpiralGrowthMask}
  )
  n4 = NamedRerouteDeclaration(None=MaterialFunctionCall!(TexCoord=TextureCoordinate()) {materialFunction=&/MF_SpiralClouds_ThicknessPattern}.Result) {name=Thickness Pattern, nodeColor=<rgba 0 0.9062 1 1>}
  n5 = Multiply(
   A=LinearInterpolate(
    A=ScalarParameter() {defaultValue=2000, parameterName=Spline Width 1}
    B=ScalarParameter() {defaultValue=12500, parameterName=Spline Width 2}
    Alpha=Max(A=@n4) {constB=0.5}
   )
   B=VertexColor().G
  )
  n6 = NamedRerouteUsage()
  NamedRerouteDeclaration(None=Multiply(A=@n5, B=@n6)) {name=Tangent Offset, nodeColor=<rgba 0 1 0.4375 1>}
  n7 = CameraVectorWS()
  n8 = Saturate(None=DotProduct(A=Multiply(A=@n7) {constB=-1}, B=SkyAtmosphereLightDirection()))
  n9 = Multiply(A=Multiply(A=Multiply(A=@n8, B=@n8), B=NamedRerouteUsage()), B=LinearInterpolate(Alpha=NamedRerouteUsage()) {constA=0.4})
  NamedRerouteDeclaration(None=Multiply(A=Multiply(A=@n9, B=@n9)) {constB=1.5}) {name=Halo Mask, nodeColor=<rgba 0 0.7656 1 1>}
  n10 = Multiply(A=Add(A=TextureCoordinate()) {constB=-0.5}) {constB=0.8}
  n11 = ComponentMask(None=@n10) {r=true}
  n12 = ComponentMask(None=@n10) {g=true}
  n13 = NamedRerouteUsage()
  n14 = Add(A=Multiply(A=@n11) {constB=-3}, B=Multiply(A=Divide(A=@n12) {constB=5000}, B=VertexInterpolator(VS=@n13).PS))
  n15 = AppendVector(A=@n11, B=@n14)
  n16 = AppendVector(A=VertexColor().R, B=Constant() {r=1})
  n17 = Multiply(A=Constant2Vector() {g=0.8, r=15}, B=@n16)
  n18 = Time()
  n19 = Constant2Vector() {g=0.1, r=-0.14}
  n20 = TextureSample(UVs=Add(A=Multiply(A=@n15, B=@n17), B=Multiply(A=@n18, B=@n19))) {samplerType=SAMPLERTYPE_LinearColor, texture=/Game/TimeOfDay/TODM_Sequence/Materials/Sky/Textures/T_Ch6_OvercastClouds}
  n21 = Constant2Vector() {g=0.5, r=10}
  n22 = Add(A=Multiply(A=AppendVector(A=@n11, B=@n14), B=Multiply(A=@n21, B=@n16)), B=Multiply(A=Time(), B=Constant2Vector() {g=0.05, r=0.03}))
  n23 = TextureSample(UVs=@n22) {samplerType=SAMPLERTYPE_LinearColor, texture=&T_Ch6_OvercastClouds}
  n24 = Multiply(A=Subtract(A=@n23.B, B=@n20.B)) {constB=4}
  n25 = LinearInterpolate(A=@n20.RGBA, B=@n23.RGBA, Alpha=Saturate(None=Add(A=@n24) {constB=0.5}))
  n26 = ComponentMask(None=@n25) {g=true, r=true}
  n27 = DeriveNormalZ(InXY=Multiply(A=Add(A=@n26) {constB=-0.5}) {constB=2})
  n28 = Multiply(A=NamedRerouteUsage()) {constB=-1}
  n29 = ComponentMask(None=TextureCoordinate()) {g=true}
  n30 = Multiply(A=Sign(None=Add(A=@n29) {constB=-0.5}), B=NamedRerouteUsage())
  n31 = NamedRerouteUsage()
  n32 = VertexInterpolator(VS=LinearInterpolate(A=NamedRerouteUsage(), B=Multiply(A=@n30) {constB=-1}, Alpha=@n31))
  n33 = NamedRerouteUsage()
  n34 = NamedRerouteUsage()
  n35 = Normalize(VectorInput=VertexInterpolator(VS=LinearInterpolate(A=@n33, B=@n34, Alpha=@n31)).PS)
  n36 = NamedRerouteUsage()
  NamedRerouteDeclaration(None=LinearInterpolate(
    A=MaterialFunctionCall(VectorToTransform=@n27, BasisX=@n28, BasisY=@n32.PS, BasisZ=@n35) {materialFunction=&Transform3x3Matrix}.Transformed Vector
    B=Multiply(A=@n36) {constB=2}
   )) {name=Normals, nodeColor=<rgba 0.5 1 0 1>}
  n37 = ComponentMask(None=TextureCoordinate()) {r=true}
  n38 = Multiply(A=@n37, B=OneMinus(None=@n37))
  NamedRerouteDeclaration(None=Power(Base=Multiply(A=@n38) {constB=2}) {constExponent=0.5}) {name=Length Mask, nodeColor=<rgba 0 1 0.1797 1>}
  n39 = Normalize(VectorInput=WorldPosition() {worldPositionShaderOffset=WPT_CameraRelative}.XYZ)
  n40 = Multiply(A=NamedRerouteUsage()) {constB=-1}
  NamedRerouteDeclaration(None=VertexInterpolator(VS=DotProduct(A=@n39, B=@n40)).PS) {name=Fresnel, nodeColor=<rgba 0 0.0625 1 1>}
  n41 = ComponentMask(None=@n25) {b=true}
  n42 = OneMinus(None=@n41)
  NamedRerouteDeclaration(None=Multiply(A=Power(Base=@n42) {constExponent=2.5}) {constB=4}) {name=Alpha, nodeColor=<rgba 0.5469 1 0 1>}
  n43 = ComponentMask(None=TextureCoordinate()) {g=true}
  n44 = Multiply(A=Abs(None=Add(A=@n43) {constB=-0.5})) {constB=2}
  n45 = OneMinus(None=NamedRerouteDeclaration(None=Multiply(A=@n44, B=@n44)) {name=V Mask, nodeColor=<rgba 0.2891 1 0 1>})
  n46 = Add(A=@n41, B=Multiply(A=@n45) {constB=0.15})
  n47 = Saturate(None=LinearInterpolate(Alpha=NamedRerouteUsage()) {constA=0.5})
  n48 = Saturate(None=NamedRerouteUsage())
  n49 = Multiply(A=Multiply(A=Multiply(A=@n46, B=@n45), B=Multiply(A=@n47, B=@n48))) {constB=3}
  NamedRerouteDeclaration(None=Saturate(None=Power(Base=@n49) {constExponent=1.3})) {name=Opacity, nodeColor=<rgba 0.7344 0 1 1>}
  MaterialExpressionWorldPosition_3 = WorldPosition() {worldPositionShaderOffset=WPT_ExcludeAllShaderOffsets}
  MaterialExpressionComponentMask_1 = ComponentMask(None=ActorPositionWS()) {g=true, r=true}
  MaterialExpressionScalarParameter_33 = ScalarParameter() {defaultValue=10000, parameterName=d}
  MaterialExpressionScalarParameter_34 = ScalarParameter() {defaultValue=62433.5859, parameterName=dr}
  n50 = StaticBool()
  n51 = StaticBoolParameter() {defaultValue=true, parameterName=QuadSpiral}
  n52 = MaterialFunctionCall!(
   PositionXY=@MaterialExpressionWorldPosition_3.XY
   SpiralCenterXY=@MaterialExpressionComponentMask_1
   SpiralRotation=ScalarParameter() {defaultValue=1, parameterName=a}
   SpiralCurve=ScalarParameter() {defaultValue=0.05, parameterName=b}
   Thickness=@MaterialExpressionScalarParameter_33
   Radius=@MaterialExpressionScalarParameter_34
   QuadSpiral=QualitySwitch(Default=@n50, Low=@n50, High=@n51, Medium=@n50, Epic=@n51)
  ) {materialFunction=&/MF_SpiralMask_WS}
  n53 = Custom!(
   p=@MaterialExpressionWorldPosition_3.XY
   spos=@MaterialExpressionComponentMask_1
   r=@MaterialExpressionScalarParameter_34
   d=@MaterialExpressionScalarParameter_33
  ) {
   code="//calculate sphere mask\r\nfloat2 q = p - spos;\r\nfloat sphere = length(q) - r;\r\nsphere = 1.0f - saturate(sphere/d);\r\nreturn sphere;"
   outputType=CMOT_Float1
   showCode=true
   inputs [
    {
     inputName=p
     input {expression=&self:MaterialExpressionWorldPosition_3, mask=1, maskG=1, maskR=1, outputIndex=1}
    }
    {inputName=spos, input {expression=&self:MaterialExpressionComponentMask_1}}
    {inputName=r, input {expression=&self:MaterialExpressionScalarParameter_34}}
    {inputName=d, input {expression=&self:MaterialExpressionScalarParameter_33}}
   ]
  }
  NamedRerouteDeclaration(None=QualitySwitch(Default=@n52.Result, Low=@n53, High=@n52.Result, Medium=@n52.Result, Epic=@n52.Result)) {name=SpiralMask, nodeColor=<rgba 0 0.5547 1 1>}
  n54 = ComponentMask(None=TextureCoordinate()) {g=true}
  n55 = Multiply(A=Abs(None=Add(A=@n54) {constB=-0.5})) {constB=2}
  n56 = NamedRerouteDeclaration(None=Transform(None=Constant3Vector() {constant=<rgba -1 0 0 0>})) {name=World Tangent, nodeColor=<rgba 1 0 0.2812 1>}
  n57 = Normalize(VectorInput=CrossProduct(A=@n56, B=@n39))
  n58 = NamedRerouteDeclaration(None=CrossProduct(A=@n56, B=@n57)) {name=Flat Normal, nodeColor=<rgba 1 0 0.6094 1>}
  n59 = Add(A=ComponentMask(None=TextureCoordinate()) {g=true}) {constB=-0.5}
  n60 = Add(
   A=Multiply(A=Multiply(A=@n55, B=@n58)) {constB=-0.2}
   B=NamedRerouteDeclaration(None=Multiply(A=@n59, B=@n57)) {name=Cross World, nodeColor=<rgba 1 0.0703 0 1>}
  )
  n61 = MaterialFunctionCall() {materialFunction=&ObjectScale}
  n62 = Multiply(
   A=Multiply(A=NamedRerouteDeclaration(None=@n60) {name=Cross + FlatNormal, nodeColor=<rgba 0.7109 0 1 1>}, B=Multiply(A=@n5, B=@n61.Scale XYZ)) {constB=3000}
   B=@n6
  )
  n63 = NamedRerouteDeclaration(None=Add(A=@n62, B=MaterialFunctionCall!(WPO=@n62) {materialFunction=&/MF_SpiralClouds_WobblePattern}.Result)) {name=WPOBeforeCameraOffset, nodeColor=<rgba 0 0.2734 1 1>}
  n64 = Normalize(VectorInput=Add(A=@n63, B=WorldPosition() {worldPositionShaderOffset=WPT_CameraRelative}.XYZ))
  n65 = ScalarParameter() {defaultValue=4000, parameterName=Spline Camera Offset 2}
  NamedRerouteDeclaration(None=Subtract(A=@n63, B=Multiply(A=@n64, B=@n65) {constB=-2500})) {name=WPO, nodeColor=<rgba 1 0 0.7734 1>}
  NamedRerouteDeclaration(None=@n57) {name=Cross World Module, nodeColor=<rgba 1 0.0703 0 1>}
  NamedRerouteUsage()
  Reroute(None=@n60)
  n66 = Multiply(A=NamedRerouteUsage()) {constB=3}
  n67 = NamedRerouteUsage()
  n68 = NamedRerouteUsage()
  Reroute(None=Add(
    A=MaterialFunctionCall!(Normals=@n66, Halo Mask=@n67) {materialFunction=&/MF_SpiralClouds_Color}.Result
    B=MaterialFunctionCall!(Lightning Alpha=@n68) {materialFunction=&/MF_SpiralClouds_Lightning}.Result
   ))
  VertexInterpolator(VS=@n22)
 ]
}
