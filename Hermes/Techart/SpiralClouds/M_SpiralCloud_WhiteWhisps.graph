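# Translucent unlit material for white wisps in spiral clouds, drawn on camera-facing ribbons.
# World position offset widens each strip along tangent-cross-view, adds MF_SpiralClouds_WobblePattern and shifts it along the view ray.
# Two panning overcast-cloud samples, blended by their blue-channel difference, supply ribbon-space normals and alpha, with edge masks, distance and depth fades, and global MPC opacity.
# MF_SpiralClouds_Color, fed normals and a sun-facing halo mask, adds to MF_SpiralClouds_Lightning.
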Material /Hermes/Techart/SpiralClouds/M_SpiralCloud_WhiteWhisps
graph {
 properties {
  bAutomaticallySetUsageInEditor=false
  bTangentSpaceNormal=false
  blendMode=BLEND_Translucent
  shadingModel=MSM_Unlit
  usedShadingModels=Unlit
 }
 nodes [
  n1 = NamedRerouteDeclaration(None=Transform(None=Constant3Vector() {constant=<rgba -1 0 0 0>})) {name=World Tangent, nodeColor=<rgba 1 0 0.2812 1>}
  n2 = Normalize(VectorInput=WorldPosition() {worldPositionShaderOffset=WPT_CameraRelative}.XYZ)
  n3 = Normalize(VectorInput=CrossProduct(A=@n1, B=@n2))
  n4 = NamedRerouteDeclaration(None=CrossProduct(A=@n1, B=@n3)) {name=Flat Normal, nodeColor=<rgba 1 0 0.6094 1>}
  Add(A=Multiply(A=@n4) {constB=8000})
  n5 = NamedRerouteUsage()
  Add(
   A=MaterialFunctionCall!(Normals=Multiply(A=@n5) {constB=2}, Halo Mask=NamedRerouteUsage()) {materialFunction=&/MF_SpiralClouds_Color}.Result
   B=MaterialFunctionCall!(Lightning Alpha=NamedRerouteUsage()) {materialFunction=&/MF_SpiralClouds_Lightning}.Result
  )
  Add(A=LocalPosition().XY, B=ComponentMask(None=NamedRerouteUsage()) {g=true, r=true})
  n6 = ComponentMask(None=TextureCoordinate()) {g=true}
  n7 = Multiply(A=Abs(None=Add(A=@n6) {constB=-0.5})) {constB=2}
  n8 = ComponentMask(None=TextureCoordinate()) {g=true}
  n9 = NamedRerouteDeclaration(None=Multiply(A=Add(A=@n8) {constB=-0.5}, B=@n3)) {name=Cross World, nodeColor=<rgba 1 0.0703 0 1>}
  Add(A=Multiply(A=Multiply(A=@n7, B=@n4)) {constB=-0.2}, B=@n9)
  Constant() {r=1}
  n10 = Multiply(A=Add(A=TextureCoordinate()) {constB=-0.5}) {constB=0.8}
  n11 = ComponentMask(None=@n10) {r=true}
  n12 = ComponentMask(None=@n10) {g=true}
  n13 = NamedRerouteUsage()
  n14 = Add(A=Multiply(A=@n11) {constB=-3}, B=Multiply(A=Divide(A=@n12) {constB=5000}, B=VertexInterpolator(VS=@n13).PS))
  n15 = AppendVector(A=@n11, B=@n14)
  n16 = Constant2Vector() {g=1.4, r=7}
  n17 = Time()
  n18 = Constant2Vector() {g=0.03, r=-0.14}
  n19 = TextureSample(UVs=Add(A=Multiply(A=@n15, B=@n16), B=Multiply(A=@n17, B=@n18))) {samplerType=SAMPLERTYPE_LinearColor, texture=/Game/TimeOfDay/TODM_Sequence/Materials/Sky/Textures/T_Ch6_OvercastClouds}
  n20 = AppendVector(A=@n11, B=@n14)
  n21 = Constant2Vector() {g=1, r=5}
  n22 = Time()
  n23 = Constant2Vector() {g=-0.02, r=-0.07}
  n24 = TextureSample(UVs=Add(A=Multiply(A=@n20, B=@n21), B=Multiply(A=@n22, B=@n23))) {samplerType=SAMPLERTYPE_LinearColor, texture=&T_Ch6_OvercastClouds}
  n25 = Multiply(A=Subtract(A=@n24.B, B=@n19.B)) {constB=4}
  n26 = LinearInterpolate(A=@n19.RGBA, B=@n24.RGBA, Alpha=Saturate(None=Add(A=@n25) {constB=0.5}))
  n27 = Add(A=ComponentMask(None=@n26) {g=true, r=true}) {constB=-0.5}
  n28 = NamedRerouteUsage()
  n29 = Add(A=ComponentMask(None=TextureCoordinate()) {g=true}) {constB=-0.5}
  n30 = Multiply(A=Multiply(A=Sign(None=@n29), B=NamedRerouteUsage())) {constB=-1}
  n31 = NamedRerouteUsage()
  n32 = LinearInterpolate(A=NamedRerouteUsage(), B=NamedRerouteUsage(), Alpha=@n31)
  MaterialFunctionCall(
   VectorToTransform=DeriveNormalZ(InXY=Multiply(A=@n27) {constB=2})
   BasisX=Multiply(A=NamedRerouteUsage()) {constB=-1}
   BasisY=VertexInterpolator(VS=LinearInterpolate(A=@n28, B=@n30, Alpha=@n31)).PS
   BasisZ=Normalize(VectorInput=VertexInterpolator(VS=@n32).PS)
  ) {materialFunction=&Transform3x3Matrix}
  n33 = Multiply(A=NamedRerouteUsage()) {constB=7}
  n34 = Multiply(A=NamedRerouteUsage(), B=Saturate(None=Subtract(A=@n33) {constB=2}))
  n35 = Distance(A=WorldPosition() {worldPositionShaderOffset=WPT_CameraRelative}.XYZ, B=Constant())
  n36 = Saturate(None=Subtract(A=Divide(A=@n35) {constB=3000}) {constB=0.5})
  Multiply(
   A=DepthFade(Opacity=Multiply(A=@n34, B=@n36)) {fadeDistanceDefault=2000}
   B=CollectionParameter() {collection=/Game/Environments/Helios/TechArt/LowAltitudeClouds/V2/MPC_LowAltitudeClouds_V2, parameterName=GlobalOpacity}
  )
  n37 = ComponentMask(None=TextureCoordinate()) {r=true}
  Multiply(A=@n37) {constB=2}
  n38 = ScalarParameter() {defaultValue=14000, parameterName=Spline Width}
  n39 = NamedRerouteUsage()
  NamedRerouteDeclaration(None=Multiply(A=@n38, B=@n39)) {name=Tangent Offset, nodeColor=<rgba 0 1 0.4375 1>}
  n40 = CameraVectorWS()
  n41 = Saturate(None=DotProduct(A=Multiply(A=@n40) {constB=-1}, B=SkyAtmosphereLightDirection()))
  n42 = Multiply(A=Multiply(A=Multiply(A=@n41, B=@n41), B=NamedRerouteUsage()), B=LinearInterpolate(Alpha=NamedRerouteUsage()) {constA=0.4})
  NamedRerouteDeclaration(None=Multiply(A=Multiply(A=@n42, B=@n42)) {constB=1.5}) {name=Halo Mask, nodeColor=<rgba 0 0.7656 1 1>}
  n43 = NamedRerouteUsage()
  NamedRerouteDeclaration(None=VertexInterpolator(VS=Multiply(A=@n43)).PS) {name=Normals, nodeColor=<rgba 0.5 1 0 1>}
  NamedRerouteDeclaration() {name=World Origin, nodeColor=<rgba 0.7578 1 0 1>}
  n44 = ComponentMask(None=TextureCoordinate()) {r=true}
  n45 = OneMinus(None=@n44)
  NamedRerouteDeclaration(None=Multiply(A=Multiply(A=@n44, B=@n45)) {constB=3}) {name=Length Mask, nodeColor=<rgba 0 1 0.1797 1>}
  NamedRerouteDeclaration(None=Add(A=@n37) {constB=0.5}) {name=Thickness pattern, nodeColor=<rgba 0.3359 0 1 1>}
  n46 = Multiply(A=NamedRerouteUsage()) {constB=-1}
  NamedRerouteDeclaration(None=VertexInterpolator(VS=DotProduct(A=@n2, B=@n46)).PS) {name=Fresnel, nodeColor=<rgba 0 0.0625 1 1>}
  n47 = ComponentMask(None=@n26) {b=true}
  n48 = OneMinus(None=@n47)
  NamedRerouteDeclaration(None=Multiply(A=Power(Base=@n48) {constExponent=3}) {constB=4}) {name=Alpha, nodeColor=<rgba 0.5469 1 0 1>}
  n49 = ComponentMask(None=TextureCoordinate()) {g=true}
  n50 = Multiply(A=Abs(None=Add(A=@n49) {constB=-0.5})) {constB=2}
  n51 = Multiply(A=@n50, B=@n50)
  n52 = OneMinus(None=NamedRerouteDeclaration(None=Multiply(A=@n51, B=@n51)) {name=V Mask, nodeColor=<rgba 0.2891 1 0 1>})
  n53 = Add(A=@n47, B=Multiply(A=@n52) {constB=0.2})
  n54 = Power(Base=Multiply(A=Multiply(A=@n53, B=@n52)) {constB=0.85})
  NamedRerouteDeclaration(None=Multiply(A=Saturate(None=@n54), B=NamedRerouteUsage())) {name=Opacity, nodeColor=<rgba 0.7344 0 1 1>}
  n55 = MaterialFunctionCall() {materialFunction=&ObjectScale}
  n56 = Multiply(A=Multiply(A=@n9, B=Multiply(A=@n38, B=@n55.Scale XYZ)) {constB=3000}, B=@n39)
  n57 = Add(A=@n56, B=MaterialFunctionCall!(WPO=@n56) {materialFunction=&/MF_SpiralClouds_WobblePattern}.Result)
  NamedRerouteDeclaration(None=@n57) {name=WPO, nodeColor=<rgba 1 0 0.7734 1>}
  NamedRerouteDeclaration(None=@n3) {name=Cross World Module, nodeColor=<rgba 1 0.0703 0 1>}
  NamedRerouteUsage()
  OneMinus(None=@n7)
  n58 = ComponentMask(None=TextureCoordinate()) {r=true}
  n59 = Multiply(A=@n58, B=OneMinus(None=@n58))
  Saturate(None=Power(Base=Multiply(A=@n59) {constB=2}) {constExponent=0.3})
  n60 = Add(A=@n57, B=WorldPosition() {worldPositionShaderOffset=WPT_CameraRelative}.XYZ)
  n61 = ScalarParameter() {defaultValue=100, parameterName=Spline Camera Offset 1}
  n62 = ScalarParameter() {defaultValue=8000, parameterName=Spline Camera Offset 2}
  Subtract(A=@n57, B=Multiply(A=Normalize(VectorInput=@n60), B=LinearInterpolate(A=@n61, B=@n62, Alpha=@n39)) {constB=-2500})
 ]
}
